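; Actor that hosts the Creative Melee lock-on reticle through a widget component.
; SetReticleVisible casts the component's widget to CreativeMelee_LockOn_Widget and passes the requested visibility to its SetVisible.
; When showing, it also unhides the scene component and its children in game; hiding leaves that component untouched.
; The construction script is empty.
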(fn SetReticleVisible
  (params (Visible :type Boolean))
  (bind _ascreative_melee_lock_on_widget (Utilities|Casting|CastToCreativeMelee_LockOn_Widget (UserInterface|GetWidget (Variables|Default|GetWidgetComp))))
  (Class|CreativeMeleeLockonWidget|SetVisible _ascreative_melee_lock_on_widget Visible true)
  (if Visible
    (Development|SetHiddeninGame (Variables|Default|GetScene) false true)))

(fn ConstructionScript)
