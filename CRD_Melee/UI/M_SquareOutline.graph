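# Translucent UI material that draws a square outline, a border mask around the edges of the element.
# Two GeneratedBand calls, one with its direction switched, are multiplied into a centered square, then inverted to leave only the border.
# A scalar parameter controls border thickness: its inverse passes through a Divide to set both band widths, with a shared constant sharpness.
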Material /CRD_Melee/UI/M_SquareOutline
graph {
 properties {blendMode=BLEND_Translucent, materialDomain=MD_UI, pixelDepthOffsetMode=PDOM_Legacy, usedShadingModels=Default Lit}
 nodes [
  Constant() {r=1}
  n1 = Divide(A=OneMinus(None=ScalarParameter() {defaultValue=0.125}))
  n2 = Constant() {r=1}
  n3 = StaticBool() {value=true}
  OneMinus(None=Multiply(
    A=MaterialFunctionCall!(Width=@n1, Sharpness=@n2, Offset, Direction Switch, Compare, Input Coordinates) {materialFunction=/Engine/Functions/Engine_MaterialFunctions03/Procedurals/GeneratedBand}.Result
    B=MaterialFunctionCall(Width=@n1, Sharpness=@n2, Direction Switch=@n3) {materialFunction=&GeneratedBand}.Result
   ))
 ]
}
