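# Lock-on reticle widget for the melee system, showing a center image and an outline image, both rotated 45 degrees.
# The FadeIn animation shrinks and fades in the outline, then collapses it as the center appears.
# SetVisible, when animated, plays FadeIn forward or reversed only if the requested state differs from ReticleVisible, then updates that flag.
# Without animation it sets both images' visibility directly, leaving ReticleVisible unchanged.
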WidgetBlueprint /CRD_Melee/UI/CreativeMelee_LockOn_Widget
animations [FadeIn {
  binding_count=2
  length_seconds=0.2333
  object=&self:FadeIn
  sequence {
   bindings [
    ReticleOutlineBorder {
     id=#1
     possessed_class=Image
     tracks [
      {
       type=MovieScene2DTransformTrack
       sections [{
         type=MovieScene2DTransformSection
         end_frame=7
         start_frame=0
         channels [
          Scale.X {type=MovieSceneScriptingFloatChannel, keys [[0,6],[7,1]]}
          Scale.Y {type=MovieSceneScriptingFloatChannel, keys [[0,6],[7,1]]}
         ]
        }]
      }
      {
       type=MovieSceneFloatTrack
       display_name=Render Opacity
       sections [{type=MovieSceneFloatSection, end_frame=7, start_frame=0, channels [None {type=MovieSceneScriptingFloatChannel, keys [[0,0],[7,1]]}]}]
      }
      {
       type=MovieSceneByteTrack
       display_name=Visibility
       sections [{
         type=MovieSceneByteSection
         end_frame=7
         start_frame=0
         channels [None {type=MovieSceneScriptingByteChannel, keys [[0,1],[1,4],[6,4],[7,1]]}]
        }]
      }
     ]
    }
    ReticleCenter {
     id=#2
     possessed_class=Image
     tracks [{
       type=MovieSceneByteTrack
       display_name=Visibility
       sections [{type=MovieSceneByteSection, end_frame=7, start_frame=0, channels [None {type=MovieSceneScriptingByteChannel, keys [[0,1],[6.9995,4]]}]}]
      }]
    }
   ]
   display_rate {numerator=30}
  }
 }]
graphs [
 EventGraph {links [], nodes []}
 SetVisible {nodes [
   n1 = SetVisible!() -> (then, Visible:bool=~, Animate:bool=~) {is=title K2Node_FunctionEntry}
   > n2 = Branch(Condition=@n1.Animate)
   > Branch(Condition="Not Equal (Boolean)"(A=Get ReticleVisible() -> (ReticleVisible:bool), B=Get Visible() -> (Visible:bool)) -> (ReturnValue=~))
   > PlayAnimation(
    InAnimation=Get FadeIn() -> (FadeIn:Widget Animation Object Reference)
    PlayMode=Select(Option 0:EUMGSequencePlayMode Enum=Reverse, Option 1:EUMGSequencePlayMode Enum, Index=Get Visible()) -> (ReturnValue:EUMGSequencePlayMode Enum)
   )
   > Set ReticleVisible(ReticleVisible:bool=Get Visible()) -> (Output_Get:bool)
   SetVisibility(
    execute=@n2.else
    self=[Get ReticleCenter() -> (ReticleCenter:Image Object Reference), Get ReticleOutlineBorder() -> (ReticleOutlineBorder:Image Object Reference)]
    InVisibility=Select(Option 0:ESlateVisibility Enum=Collapsed, Option 1:ESlateVisibility Enum=SelfHitTestInvisible, Index=Get Visible()) -> (ReturnValue:ESlateVisibility Enum)
   )
  ]}
]
variables [ReticleVisible {type=boolean}]
widgets {
 info {inheritedWidgetCount=0, parentClass=&UserWidget, rootWidgetClass=&Overlay, widgetCount=3}
 widgets [
  Overlay_16 {widgetClassPath=&Overlay, widget=Overlay_16}
  ReticleCenter {
   widgetClassPath=&Image
   bIsVariable=true
   parent=Overlay_16
   widget=ReticleCenter
   properties {
    visibility=Collapsed
    brush {
     bIsDynamicallyLoaded=false
     drawAs=Image
     imageType=NoImage
     mirroring=NoMirror
     resourceName=None
     resourceObject=null
     tiling=NoTile
     imageSize=<16 16>
     margin=<ltrb 0 0 0 0>
     outlineSettings {
      bUseBrushTransparency=false
      roundingType=HalfHeightRadius
      width=0
      color {colorUseRule=UseColor_Specified, specifiedColor=<rgba 0 0 0 0>}
      cornerRadii=<0 0 0 0>
     }
     tintColor {colorUseRule=UseColor_Specified, specifiedColor=<rgba 1 1 1 1>}
     uVRegion {bIsValid=false, max=<0 0>, min=<0 0>}
    }
    renderTransform {angle=45, scale=<1 1>, shear=<0 0>, translation=<0 0>}
   }
   slot {class=OverlaySlot}
  }
  ReticleOutlineBorder {
   widgetClassPath=&Image
   bIsVariable=true
   parent=Overlay_16
   widget=ReticleOutlineBorder
   properties {
    renderOpacity=0
    visibility=Collapsed
    brush {
     bIsDynamicallyLoaded=false
     drawAs=Box
     imageType=NoImage
     mirroring=NoMirror
     resourceName=None
     resourceObject=&/M_SquareOutline
     tiling=NoTile
     imageSize=<16 16>
     margin=<ltrb 0.5 0.5 0.5 0.5>
     outlineSettings {
      bUseBrushTransparency=false
      roundingType=HalfHeightRadius
      width=0
      color {colorUseRule=UseColor_Specified, specifiedColor=<rgba 0 0 0 0>}
      cornerRadii=<0 0 0 0>
     }
     tintColor {colorUseRule=UseColor_Specified, specifiedColor=<rgba 1 1 1 1>}
     uVRegion {bIsValid=false, max=<0 0>, min=<0 0>}
    }
    renderTransform {angle=45, scale=<8 8>, shear=<0 0>, translation=<0 0>}
   }
   slot {class=OverlaySlot}
  }
 ]
}
